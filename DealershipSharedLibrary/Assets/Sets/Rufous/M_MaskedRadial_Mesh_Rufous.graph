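# Unlit masked material for Niagara mesh particles, apparently a smoke-like radial effect with glowing fresnel and inner emission.
# Panning base texture, fresnel and a V gradient are offset, sharpened and multiplied by particle alpha into a cutout mask.
# Panning noise and a V band form a hard-stepped inner emission, added to a noise-modulated fresnel glow through MF_FX_EmissiveAdjustments.
# Four dynamic parameters expose scales, offsets, inversion, opacity and emission controls per particle.
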Material /DealershipSharedLibrary/Assets/Sets/Rufous/M_MaskedRadial_Mesh_Rufous
graph {
 properties {
  bAutomaticallySetUsageInEditor=false
  bGenerateSphericalParticleNormals=true
  bUsedWithNiagaraMeshParticles=true
  blendMode=BLEND_Masked
  refractionMethod=RM_IndexOfRefraction
  shadingModel=MSM_Unlit
  usedShadingModels=Unlit
 }
 nodes [
  n1 = DynamicParameter() {parameterIndex=2, defaultValue=<rgba 0 1 0 0.45>, paramNames [InvertMask, NoiseScale_Y, NoiseOffset_X, NoiseOffset_Y]}
  n2 = Multiply(A=TextureCoordinate(), B=AppendVector(A=ScalarParameter() {defaultValue=1, parameterName=NoiseScale_X}, B=@n1.NoiseScale_Y))
  n3 = AppendVector(A=@n1.NoiseOffset_X, B=@n1.NoiseOffset_X)
  Add(A=@n2, B=@n3)
  n4 = ScalarParameter() {defaultValue=0.2, parameterName=Emission Fresnel Exponentin}
  n5 = Saturate(None=OneMinus(None=Fresnel(ExponentIn=@n4) {baseReflectFraction=0.1, exponent=0.524}))
  n6 = TextureCoordinate()
  n7 = ComponentMask(None=@n6) {g=true}
  n8 = TextureSampleParameter2D(UVs=Panner(Coordinate=@n2, Speed=@n3)) {
   parameterName=Noise Texture
   samplerType=SAMPLERTYPE_Grayscale
   texture=/Game/Effects/Fort_Effects/Textures/Plasma/T_Turbulence_Stretched_Balanced
  }
  n9 = LinearInterpolate(A=Clamp(None=ConstantBiasScale(None=@n7) {bias=-0.2, scale=1}), B=Clamp(None=@n8.RGB), Alpha=@n7)
  n10 = Multiply(A=Multiply(A=@n5, B=Clamp(None=@n9) {maxDefault=10.9281, minDefault=-0.32}))
  n11 = Multiply(A=Saturate(None=Add(A=@n5, B=@n10)), B=NamedRerouteUsage())
  n12 = ParticleColor()
  n13 = LinearInterpolate(A=VectorParameter() {parameterName=FresnelColor, defaultValue=<rgba 0.292 1 0 1>}.RGB, B=@n12.RGB, Alpha=Clamp(None=NamedRerouteUsage()))
  n14 = Multiply(A=@n12.RGB, B=NamedRerouteUsage())
  n15 = VectorParameter() {parameterName=EmissionTint, defaultValue=<rgba 1 1 0 1>}
  n16 = Clamp(None=NamedRerouteUsage())
  n17 = Multiply(A=Multiply(A=@n12.RGB, B=LinearInterpolate(A=@n15.RGB, Alpha=@n16)), B=ScalarParameter() {defaultValue=1, parameterName=MultiplyEmission})
  n18 = Constant()
  n19 = Constant() {r=-0.8}
  n20 = NamedRerouteUsage()
  n21 = Constant()
  n22 = DynamicParameter() {parameterIndex=3, defaultValue=<rgba 0.4 0 0 1>, paramNames [InnerEmissionTextureLerp, IsPainted, InvertEmission, Y offset]}
  n23 = ComponentMask(None=Add(A=@n6, B=AppendVector(A=@n21, B=@n22.Y offset))) {g=true}
  n24 = Frac(None=@n23)
  n25 = OneMinus(None=@n24)
  n26 = Add(
   A=LinearInterpolate(A=Constant() {r=-0.3}, B=Constant() {r=-1}, Alpha=ScalarParameter() {defaultValue=0.0898, parameterName=tile})
   B=ConstantBiasScale(None=Multiply(A=@n25, B=@n24)) {bias=-0.01, scale=3}
  )
  n27 = NamedRerouteUsage()
  n28 = Multiply(A=Add(A=LinearInterpolate(A=@n18, B=@n19, Alpha=@n20), B=LinearInterpolate(A=@n26, B=@n9, Alpha=@n27) {constAlpha=0.4})) {constB=8}
  n29 = Clamp(None=OneMinus(None=Ceil(None=@n28)))
  n30 = LinearInterpolate(A=@n29, B=OneMinus(None=@n29), Alpha=NamedRerouteUsage())
  MaterialFunctionCall(Emissive=Add(A=Multiply(A=@n11, B=@n13), B=LinearInterpolate(A=@n14, B=@n17, Alpha=@n30))) {materialFunction=&MF_FX_EmissiveAdjustments}
  n31 = OneMinus(None=Multiply(A=@n23) {constB=32})
  n32 = DynamicParameter() {parameterIndex=1, defaultValue=<rgba 1 1 0 0.879>, paramNames [BaseScale_X,BaseScale_Y,BaseOffset_X,BaseOffset_Y]}
  n33 = Multiply(A=TextureCoordinate(), B=AppendVector(A=@n32.BaseScale_X, B=@n32.BaseScale_X))
  n34 = AppendVector(A=@n32.BaseOffset_X, B=@n32.BaseOffset_X)
  n35 = TextureSampleParameter2D(UVs=Panner(Coordinate=Add(A=@n33, B=@n34), Speed=@n34)) {
   parameterName=Base Texture
   samplerType=SAMPLERTYPE_Grayscale
   texture=&T_Turbulence_Stretched_Balanced
  }
  n36 = OneMinus(None=StaticSwitchParameter(True=ChannelMaskParameter(None=@n35.RGBA) {group=Base Texture, parameterName=Base Texture Channel, sortPriority=11}.R, False=@n35.RGB) {parameterName=Use Texture Channel, sortPriority=10})
  n37 = Fresnel(ExponentIn=ScalarParameter() {defaultValue=1, parameterName=MaskFresnelExponent}) {baseReflectFraction=0, exponent=1}
  n38 = LinearInterpolate(A=OneMinus(None=@n37), B=@n37, Alpha=NamedRerouteUsage())
  n39 = Subtract(A=Add(A=Clamp(None=@n31), B=Add(A=@n36, B=@n38)), B=LinearInterpolate(A=Constant() {r=-0.2}, B=Constant() {r=3}, Alpha=NamedRerouteUsage()))
  Multiply(A=@n12.A, B=Clamp(None=Multiply(A=@n39) {constB=8}))
  NamedRerouteDeclaration(None=@n22.InnerEmissionTextureLerp) {name=InnerEmissionTextureLerp, nodeColor=<rgba 0.989 0 1 1>}
  NamedRerouteDeclaration(None=@n22.InvertEmission) {name=InvertEmission, nodeColor=<rgba 0.989 0 1 1>}
  n40 = DynamicParameter() {defaultValue=<rgba 0.1058 0.36 0 1>, paramNames [InnerEmissionSize, Opacity, FresnelEmissionAmount, SmokeDarkness]}
  NamedRerouteDeclaration(None=@n40.InnerEmissionSize) {name=InnerEmissionSize, nodeColor=<rgba 0.6875 0 1 1>}
  NamedRerouteDeclaration(None=@n40.SmokeDarkness) {name=SmokeDarkness, nodeColor=<rgba 0.9709 0.0368 1 1>}
  NamedRerouteDeclaration(None=@n40.FresnelEmissionAmount) {name=FresnelEmissionAmount, nodeColor=<rgba 1 0 0.8302 1>}
  NamedRerouteDeclaration(None=@n40.Opacity) {name=Opacity, nodeColor=<rgba 1 0 0.9845 1>}
  NamedRerouteDeclaration(None=@n22.IsPainted) {name=blsIsPainted, nodeColor=<rgba 0.1953 0 1 1>}
  NamedRerouteDeclaration(None=@n1.InvertMask) {name="InvertMask(0-1)", nodeColor=<rgba 0.4914 0.0591 1 1>}
 ]
}
